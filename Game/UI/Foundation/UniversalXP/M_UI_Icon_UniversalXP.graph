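# UI material for the Universal XP icon, drawn in the UI domain with alpha-composite blending.
# The shared MF_UI_UniversalXP_Base function supplies the color, which is multiplied by itself.
# That squared result is scaled by vertex alpha, so the widget's alpha fades it.
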Material /Game/UI/Foundation/UniversalXP/M_UI_Icon_UniversalXP
graph {
 properties {
  bAutomaticallySetUsageInEditor=false
  bUsedWithStaticMesh=false
  blendMode=BLEND_AlphaComposite
  materialDomain=MD_UI
  usedShadingModels=Default Lit
 }
 nodes [
  n1 = MaterialFunctionCall() {materialFunction=/Game/UI/Foundation/UniversalXP/MaterialFunction/MF_UI_UniversalXP_Base}
  Multiply(A=VertexColor().A, B=Multiply(A=@n1.Color, B=@n1.Color))
 ]
}
